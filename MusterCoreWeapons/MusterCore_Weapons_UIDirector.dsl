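; Dynamic UI Director for MusterCore weapons that shows or hides the red dot reticle HUD scene based on the equipped weapon.
; On BeginPlay, if in zone, it binds to the HUD context's weapon equipped and unequipped delegates.
; Equipping checks the weapon's scope interface for the RedDotReticle gameplay tag to add or remove the scene; unequipping removes it.
; A boolean tracks scene state to avoid duplicate adds or removes.
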(event EventTick
  (params (DeltaSeconds :type Float%20%28single-precision%29)))

(event EventBeginPlay
  (bind _returnvalue (BlueprintContext|GetFortHUDContext))
  (if (GlobalUIContext|IsinZone (PlayerController|LocalPlayerSubsystems|GetFortGlobalUIContext (DynamicUIDirector|GetOwningLocalPlayerController)))
    (HUD|AssignOnWeaponEquippedDelegate _returnvalue (delegate Custom|OnWeaponEquippedDelegate_Event))
    (HUD|AssignOnWeaponUnEquippedDelegate _returnvalue (delegate Custom|OnWeaponUnEquippedDelegate_Event))))

(event Custom|OnWeaponEquippedDelegate_Event
  (params (NewWeapon :type Fort%20Weapon%20Object%20Reference) (PrevWeapon :type Fort%20Weapon%20Object%20Reference))
  (bind _reddotreticlesceneactive (Variables|Default|GetRedDotReticleSceneActive))
  (bind _returnvalue (Class|FortWeaponScopeInterface|HasScopeHUDReticleTag(Message) NewWeapon (GameplayTags|MakeLiteralGameplayTag "(TagName=\"HUD.Weapon.Reticles.RedDotReticle\")")))
  (if _returnvalue
    (if (not _reddotreticlesceneactive)
      (DynamicUIDirector|AddScene "/MusterCoreWeapons/MustCore_RedDotReticle_Scene.MustCore_RedDotReticle_Scene")
      (Variables|Default|SetRedDotReticleSceneActive true))
    (elif _reddotreticlesceneactive
      (DynamicUIDirector|RemoveScene "/MusterCoreWeapons/MustCore_RedDotReticle_Scene.MustCore_RedDotReticle_Scene")
      (Variables|Default|SetRedDotReticleSceneActive false))))

(event Custom|OnWeaponUnEquippedDelegate_Event
  (if (Variables|Default|GetRedDotReticleSceneActive)
    (DynamicUIDirector|RemoveScene "/MusterCoreWeapons/MustCore_RedDotReticle_Scene.MustCore_RedDotReticle_Scene")
    (Variables|Default|SetRedDotReticleSceneActive false)))

(fn ConstructionScript)
