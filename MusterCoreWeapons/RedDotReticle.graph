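# RedDotReticle is a HUD element widget showing a red dot reticle image while the player's equipped weapon is targeting.
# On construct it binds to the HUD context's scale and targeting events and runs both handlers once.
# The targeting handler shows the dot only for a valid, targeting weapon, otherwise collapsing it.
# The scale handler sets the ScaleBox to the inverse HUD scale and doubles the dot's render scale on mobile.
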WidgetBlueprint /MusterCoreWeapons/RedDotReticle
graphs [
 EventGraph {nodes [
   n1 = OnHudScaleChanged() {is=title K2Node_CustomEvent}
   > HUDScaleChanged()
   n2 = OnPlayerTargetingChanged!() -> (OutputDelegate, then, bIsTargeting:bool=~) {is=title K2Node_CustomEvent}
   > PlayerTargetingChanged()
   Event Construct()
   n3 = Get FortHUDContext!() -> (ReturnValue:Fort HUDContext Object Reference) {is=title K2Node_GetBlueprintContextEx}
   > Bind Event to On HUDScale Changed(self:Fort HUDContext Object Reference=@n3, Delegate=@n1) {is=title K2Node_AddDelegate}
   > HUDScaleChanged()
   > Bind Event to On Player Targeting Changed(self:Fort HUDContext Object Reference=[Get FortHUDContext(), @n3], Delegate=@n2) {is=title K2Node_AddDelegate}
   > PlayerTargetingChanged()
  ]}
 HUDScaleChanged {nodes [
   HUDScaleChanged() {is=title K2Node_FunctionEntry}
   n1 = Get FortSettingsContext!() -> (ReturnValue:Fort Settings Context Object Reference) {is=title K2Node_GetBlueprintContextEx}
   > SetUserSpecifiedScale!(
    execute
    self:Scale Box Object Reference=Get ScaleBox_39() -> (ScaleBox_39:Scale Box Object Reference)
    InUserSpecifiedScale:float=SafeDivide(A=1, B=GetHUDScale!(self:Fort Settings Context Object Reference=@n1) -> (ReturnValue:float))
   ) -> (then)
   > SetRenderScale!(
    execute
    self:Widget Object Reference=Get Dot() -> (Dot:Image Object Reference)
    Scale:Vector 2D Structure=Select(Option 0:Vector 2D Structure=(X=1,Y=1), Option 1:Vector 2D Structure=(X=2,Y=2), Index=IsMobileGame()) -> (ReturnValue:Vector 2D Structure)
   ) -> (then)
  ]}
 PlayerTargetingChanged {nodes [
   PlayerTargetingChanged() {is=title K2Node_FunctionEntry}
   n1 = GetEquippedWeapon!(self:Fort Inventory Context Object Reference=Get FortInventoryContext!() -> (ReturnValue:Fort Inventory Context Object Reference) {is=title K2Node_GetBlueprintContextEx}) -> (ReturnValue:Fort Weapon Object Reference)
   > n2 = Is Valid(InputObject=@n1)
   > n3 = Branch(Condition=GetIsTargeting!(self:Fort Weapon Object Reference=@n1) -> (ReturnValue:bool))
   n4 = Get Dot()
   > SetVisibility(self=@n4, InVisibility=HitTestInvisible)
   SetVisibility(execute=[@n3.else, @n2.Is Not Valid], self=@n4, InVisibility=Collapsed)
  ]}
]
variables []
widgets {
 info {inheritedWidgetCount=1, parentClass=&FortHUDElementWidget, rootWidgetClass=&ScaleBox, widgetCount=3}
 widgets [
  ScaleBox_39 {widgetClassPath=&ScaleBox, bIsVariable=true, widget=ScaleBox_39, properties {stretch=UserSpecified}}
  Dot {
   widgetClassPath=&Image
   bIsVariable=true
   parent=ScaleBox_39
   widget=Dot
   properties {
    visibility=Collapsed
    brush {
     bIsDynamicallyLoaded=false
     drawAs=Image
     imageType=NoImage
     mirroring=NoMirror
     resourceName=None
     resourceObject=/Game/Athena/HUD/reticles/T_UI_reticle_ReddotB
     tiling=NoTile
     imageSize=<16 16>
     margin=<ltrb 0 0 0 0>
     outlineSettings {
      bUseBrushTransparency=false
      roundingType=HalfHeightRadius
      width=0
      color {colorUseRule=UseColor_Specified, specifiedColor=<rgba 0 0 0 0>}
      cornerRadii=<0 0 0 0>
     }
     tintColor {colorUseRule=UseColor_Specified, specifiedColor=<rgba 1 1 1 1>}
     uVRegion {bIsValid=false, max=<0 0>, min=<0 0>}
    }
   }
   slot {class=ScaleBoxSlot, properties {horizontalAlignment=HAlign_Left, verticalAlignment=VAlign_Top}}
  }
  TutorialHighlight {widgetClassPath=&TutorialHighlight}
 ]
}
